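# Niagara idle effect for the FastCheetah pickaxe that trails faint white, upward-drifting ribbons from its skeletal mesh.
# FN_AttachedRibbon spawns two-renderer ribbons that partly follow source movement early in life, with width curves and per-ribbon randomized, panning, eroding material.
# Opacity fades with distance from the sampled mesh point, ramps in with system age, follows the timeline and is dimmed in the front end.
# A sprite-based SmokeStationary emitter is present but disabled.
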NiagaraSystem /BRCosmetics/Weapons/FORT_Melee/Pickaxe_FastCheetah/FX/NS_Pickaxe_FastCheetah_Idle
dependencies {
 usedDataInterfaces [&NiagaraDataInterfaceCurve, &NiagaraDataInterfaceColorCurve]
 usedDynamicInputs []
 usedModules [
  &SystemState
  &self:SystemSpawnScript.SystemScriptSource.SystemScriptGraph.NiagaraNodeAssignment_1.SetVariables_#1
  &Timeline
 ]
}
emitters {
 FN_AttachedRibbon {
  emitterUpdateScript [
   &EmitterState()
   &FN_SpawnRibbon_v2(SpawnRate=30)
   &FN_Integrator() {enabled=false}
  ]
  particleSpawnScript [
   /Game/Effects/Niagara/Modules/Ribbons/FN_InitializeRibbon_v4(Enable Skeletal Position=true, Lifetime=1, Ribbon Width=32, SkeletalMesh:NiagaraDataInterfaceSkeletalMesh=$User.SkeletalMesh)
   /SetVariables(Particles.RandomFromRibbonID:NiagaraFloat=/Game/Effects/Niagara/DynamicInputs/Random/FN_SeededRandomRangeFloat(Minimum:NiagaraFloat=0, Maximum:NiagaraFloat=1, Random Seed:NiagaraInt32=$Particles.SpawnGroup)) {isSetParameters=true}
   &AddVelocity(Coordinate Space=::World, Velocity=<0 0 50>)
   &ApplyInitialForces()
   &SkeletalMeshLocation(
    Apply To Particle Position=false
    Inherit Velocity=false
    MaskSpawnGroup=true
    Mesh Sampling Type=::NewEnumerator5
    Sampled Position Offset Coordinate Space:ENiagaraMeshTransforms=::NewEnumerator3
    Skeletal Mesh=$System.Skeletal Mesh
    Socket Sampling Mode=::NewEnumerator5
   )
  ]
  particleUpdateScript [
   /SetVariables(Particles.Lifetime=&FloatFromCurve(FloatCurve=DataInterface() {propertyValues {Curve {keys [<at 0 1>,<at 1 0.2>]}}}, CurveIndex=$System.FN_Integrator.Value, Scale Curve=1.5)) {isSetParameters=true}
   &SkeletalMeshLocation(
    Apply To Particle Position=false
    Inherit Velocity=false
    MaskSpawnGroup=false
    Mesh Sampling Type=::NewEnumerator5
    Skeletal Mesh=$System.Skeletal Mesh
    Socket Sampling Mode=::NewEnumerator5
   ) {module=SkeletalMeshLocation001}
   /SetVariables(Particles.DistanceFromSource:NiagaraFloat=&ClampFloat(Float=&Divide_Float(
      A=&Add_Float(
       A=0
       B=&VectorLength(Vector=&Subtract_Vector(A=Linked() {linkedVariable {name=Particles.Position, type=NiagaraPosition}}, B=$Output.SkeletalMeshLocation001.SampledPosition))
      )
      B=100
     ))) {isSetParameters=true}
   &InheritSourceMovement(
    Applied Position Scale=&VectorFromFloat(Value=&FloatFromCurve(FloatCurve=DataInterface() {propertyValues {Curve {keys [{arriveTangent=-3.6922, interpMode=RCIM_Cubic, leaveTangent=-3.6922, tangentMode=RCTM_User, time=0, value=0.8}, <at 0.7 0 RCIM_Cubic>]}}}))
    ApplyForce=false
    ApplyPosition=true
    ApplyPositionFalloff=false
   )
   &SolveForcesAndVelocity(Clamp Velocity=false, Limit Acceleration=false)
   /Game/Effects/Niagara/Modules/Ribbons/FN_UpdateRibbon_v4(Enable Skeletal Position=true, SkeletalMesh:NiagaraDataInterfaceSkeletalMesh=$User.SkeletalMesh)
   &FN_RibbonMaterial(
    Color=&ColorFromCurve(ColorCurve=DataInterface() {propertyValues {
       AlphaCurve {keys [
         {arriveTangent=0.3686, interpMode=RCIM_Cubic, leaveTangent=0.3686, tangentMode=RCTM_User, time=0, value=0.0025}
         <at 0.2 0.0375>
         <at 1 0 RCIM_Cubic>
        ]}
       BlueCurve {keys [<at 0 1 RCIM_Cubic>]}
       GreenCurve {keys [<at 0 1 RCIM_Cubic>]}
       RedCurve {keys [<at 0 1 RCIM_Cubic>]}
      }})
    Erosion Amount:NiagaraFloat=&FloatFromCurve(
     FloatCurve=DataInterface() {propertyValues {Curve {keys [{arriveTangent=5.1387, interpMode=RCIM_Cubic, leaveTangent=5.1387, tangentMode=RCTM_User, time=0, value=0}, <at 1 1.5 RCIM_Cubic>]}}}
     Scale Curve=0.5
    )
    Erosion Pan Speed U=1
    Opacity Pan U Offset:NiagaraFloat=$Particles.RandomFromRibbonID
    Opacity Pan U Speed=0.5
    Scale Alpha:NiagaraFloat=&FloatFromCurve(
     FloatCurve=DataInterface() {propertyValues {Curve {keys [
         {interpMode=RCIM_Cubic, tangentMode=RCTM_User, time=0, value=1}
         {arriveTangent=-2.9618, interpMode=RCIM_Cubic, leaveTangent=-2.9618, tangentMode=RCTM_User, time=1, value=0}
        ]}}}
     CurveIndex=$Particles.DistanceFromSource
     Scale Curve=&Multiply_Float(
      A=$System.Timeline.InverseNormalizedTime
      B=&FloatFromCurve(
       FloatCurve=DataInterface() {propertyValues {Curve {keys [{interpMode=RCIM_Cubic, tangentMode=RCTM_User, time=0, value=0}, {interpMode=RCIM_Cubic, tangentMode=RCTM_User, time=1.5, value=1}]}}}
       CurveIndex=$Engine.System.Age
       Scale Curve=&FN_FrontEndFloatModifier(Gameplay=1, Front End=0.7)
      )
     )
    )
   ) {module=FN_RibbonMaterial001}
   &ScaleRibbonWidth(Ribbon Width Scale=&FloatFromCurve(FloatCurve=DataInterface() {propertyValues {Curve {keys [
         {arriveTangent=4.1493, interpMode=RCIM_Cubic, leaveTangent=4.1493, tangentMode=RCTM_User, time=0, value=0.7}
         {arriveTangent=-0.7, interpMode=RCIM_Cubic, leaveTangent=-0.7, time=0.1, value=1}
         {interpMode=RCIM_Cubic, tangentMode=RCTM_User, time=1, value=0}
        ]}}}))
   &KillParticles(Kill Particles=&SetBoolByFloatComparison(
     A=&MakeFloatFromLinearColor(LinearColor=$Particles.Color, Channel:ENiagaraLinearColor_Channels=::NewEnumerator3)
     B=0
     Comparison Type=::NewEnumerator4
    ))
  ]
  rendererClasses [&NiagaraRibbonRendererProperties]
  renderers [
   {rendererClass=&NiagaraRibbonRendererProperties}
   {rendererClass=&NiagaraRibbonRendererProperties, rendererIndex=1}
  ]
 }
 SmokeStationary {
  bEnabled=false
  emitterUpdateScript [
   &EmitterState()
   &SpawnRate(SpawnRate=10, Use Spawn Probability=false)
  ]
  particleSpawnScript [
   /SetVariables(Particles.SpawnGroup=&SpawnGroup()) {isSetParameters=true}
   /Niagara/Modules/Spawn/Initialization/InitializeParticle(
    Color=<rgba 1 1 1 0.05>
    Lifetime=&RandomRangeFloat(Minimum=0.8, Maximum=1.2)
    Module.Write Color=true
    Module.Write Lifetime=true
    Module.Write Mass=false
    Module.Write Material Random=true
    Module.Write Position=false
    Module.Write Scale=false
    Module.Write SpriteRotation=true
    Module.Write SpriteSize=true
    Position=&SimulationPosition()
    Sprite Rotation=&UniformRangedFloat(Minimum=0, Maximum=360)
    Sprite Size=&Vector2DFromFloat(Value=&UniformRangedFloat(Minimum=15, Maximum=25))
   ) {enabled=false}
   /Niagara/Modules/Spawn/Initialization/V2/InitializeParticle(
    Color=<rgba 1 1 1 0.05>
    Lifetime=&RandomRangeFloat(Minimum=0.8, Maximum=1.2)
    Sprite Rotation Angle=&UniformRangedFloat(Minimum=0, Maximum=360)
    Sprite Rotation Mode=::NewEnumerator0
    Sprite Size=&Vector2DFromFloat(Value=&UniformRangedFloat(Minimum=16, Maximum=24))
    Sprite Size Mode=::NewEnumerator4
    UsePositionOffset=false
    Write Material Random=true
   ) {module=InitializeParticle001}
   &SphereLocation(
    MaskSpawnGroup=false
    Override Seed=false
    Sphere Radius=5
    Surface Only=false
    UseNonUniformScale=false
    UseOffset=false
    UseSphereOrientation=false
   )
   &SkeletalMeshLocation(
    Inherit Velocity=false
    MaskSpawnGroup=true
    Mesh Sampling Type=::NewEnumerator5
    Sampled Position Offset=&Multiply_Vector(A=<1 1 1>, B=$Output.SphereLocation.SphereVector)
    Sampled Position Offset Coordinate Space:ENiagaraMeshTransforms=::NewEnumerator3
    Skeletal Mesh=$System.Skeletal Mesh
    Socket Sampling Mode=::NewEnumerator5
   )
   &AddVelocity(Velocity=&RandomRangeVector(Minimum=<0 0 2>, Maximum=<1 1 8>))
   &CameraOffset(Camera Offset Amount=5)
   &InheritVelocity(Module.Limit Speed=true, Velocity Limit=20, Velocity Scale=&VectorFromFloat(Value=0.25))
  ]
  particleUpdateScript [
   &ParticleState()
   &SkeletalMeshLocation(
    Apply To Particle Position=false
    Inherit Velocity=false
    MaskSpawnGroup=false
    Mesh Sampling Type=::NewEnumerator5
    Skeletal Mesh=$System.Skeletal Mesh
    Socket Sampling Mode=::NewEnumerator5
   ) {module=SkeletalMeshLocation001}
   /SetVariables(Particles.DistanceFromSource:NiagaraFloat=&ClampFloat(Float=&Divide_Float(
      A=&Add_Float(
       A=0
       B=&VectorLength(Vector=&Subtract_Vector(A=Linked() {linkedVariable {name=Particles.Position, type=NiagaraPosition}}, B=$Output.SkeletalMeshLocation001.SampledPosition))
      )
      B=20
     ))) {isSetParameters=true}
   &GravityForce(Gravity=<0 0 20>)
   &SolveForcesAndVelocity(Clamp Velocity=false, Limit Acceleration=false)
   &ScaleSpriteSize(Scale Factor=&Vector2DFromFloat(Value=&FloatFromCurve(FloatCurve=DataInterface() {propertyValues {Curve {keys [{arriveTangent=-0.28, interpMode=RCIM_Cubic, leaveTangent=-0.28, tangentMode=RCTM_User, time=0, value=0.5}, <at 1 1>]}}}))) {module=ScaleSpriteSize001}
   &ScaleSpriteSize(Scale Factor=&Vector2DFromFloat(Value=&FloatFromCurve(FloatCurve=DataInterface() {propertyValues {Curve {keys [
          {arriveTangent=-0.0248, interpMode=RCIM_Cubic, leaveTangent=-0.0248, tangentMode=RCTM_User, time=0, value=1}
          {arriveTangent=-0.5458, interpMode=RCIM_Cubic, leaveTangent=-0.5458, tangentMode=RCTM_User, time=1, value=0.75}
         ]}}})))
   &DynamicMaterialParameters(
    Index 0 Param 1=&FloatFromCurve(
     FloatCurve=DataInterface() {propertyValues {Curve {keys [
         {arriveTangent=4.0913, interpMode=RCIM_Cubic, leaveTangent=4.0913, tangentMode=RCTM_User, time=0, value=0}
         {arriveTangent=0.1513, interpMode=RCIM_Cubic, leaveTangent=0.1513, tangentMode=RCTM_User, time=0.5, value=1}
         {arriveTangent=-0.0082, interpMode=RCIM_Cubic, leaveTangent=-0.0082, tangentMode=RCTM_User, time=1, value=0}
        ]}}}
     Scale Curve=0.6
    )
    Index 0 Param 4=1
    Module.Param0WriteEnabled=true
    Module.Param1WriteEnabled=false
    Module.Param2WriteEnabled=false
    Module.Param3WriteEnabled=true
   )
   /SetVariables(Particles.StartFrame:NiagaraFloat=&RandomRangeFloat(Minimum=0, Maximum=63)) {isSetParameters=true}
   /SetVariables(Particles.SubImageIndex=&Lerp_Float(
     A=$Particles.StartFrame
     B=&Add_Float(A=$Particles.StartFrame, B=&RandomRangeFloat(Minimum=2, Maximum=6))
     Alpha=$Particles.NormalizedAge
    )) {isSetParameters=true}
   &Color(Color=&LerpLinearColors(StartColor=<rgba 1 0 0 1>, EndColor=<rgba 0 0 1 1>, LerpFactor=$Particles.DistanceFromSource)) {enabled=false}
   &ScaleColor(Scale Alpha=&Lerp_Float(A=1, B=0, Alpha=$Particles.DistanceFromSource), ScaleA=true, ScaleRGB=true)
   &KillParticles(Kill Particles=&SetBoolByFloatComparison(
     A=&MakeFloatFromLinearColor(LinearColor=$Particles.Color, Channel:ENiagaraLinearColor_Channels=::NewEnumerator3)
     B=0
     Comparison Type=::NewEnumerator4
    ))
  ]
  rendererClasses [&NiagaraSpriteRendererProperties]
  renderers [{rendererClass=&NiagaraSpriteRendererProperties}]
 }
}
summary {
 systemName=NS_Pickaxe_FastCheetah_Idle
 emitters [
  {bEnabled=false, emitterName=SmokeStationary, rendererClasses [&NiagaraSpriteRendererProperties]}
  {emitterName=FN_AttachedRibbon, rendererClasses [&NiagaraRibbonRendererProperties]}
 ]
 userVariables [
  User.bIsFrontEnd {type=NiagaraBool}
  User.bIsFrontendPreview {type=NiagaraBool}
  User.IsSkydiving {type=NiagaraInt32}
  User.RibbonIDOffset {type=NiagaraInt32}
  User.SkeletalMesh {
   type=NiagaraDataInterfaceSkeletalMesh
   defaultValue {
    struct=NiagaraExt_VariableValue_DataInterface
    value {dataInterface=&self:NiagaraDataInterfaceSkeletalMesh_0}
   }
  }
 ]
}
