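# Niagara melee trail for the FastCheetah pickaxe, drawing two short-lived ribbons from a skeletal mesh bone or socket through the shared AnimTrail modules.
# MainTrail has a warm time-of-day color, alpha shaped by normalized particle index and tapered width.
# Outline is translucent white, fades with a night-dimming color scale, writes dark outline color and ribbon-matched outline width, and feeds an emitter random value to a dynamic material parameter.
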NiagaraSystem /BRCosmetics/Weapons/FORT_Melee/Pickaxe_FastCheetah/FX/NS_Pickaxe_FastCheetah_Trail
dependencies {
 usedDataInterfaces [&NiagaraDataInterfaceCurve, &NiagaraDataInterfaceSkeletalMesh]
 usedDynamicInputs []
 usedModules [
  &SystemState
  &self:SystemSpawnScript.SystemScriptSource.SystemScriptGraph.NiagaraNodeAssignment_0.SetVariables_#1
 ]
}
emitters {
 MainTrail {
  emitterUpdateScript [
   &EmitterState(Life Cycle Mode=::NewEnumerator1, Loop Behavior=::NewEnumerator1, Loop Delay=0.05, Loop Duration=0.2, UseLoopDelay=false)
   &FN_SpawnAnimTrail_v2()
  ]
  particleSpawnScript [
   /Game/Effects/Niagara/Modules/AnimTrail/FN_InitializeAnimTrail/FN_InitializeAnimTrailBoneIndecies(First Bone or Socket Index=5, Second Bone or Socket Index=5)
   &FN_InitializeAnimTrail(
    AnimTrailWidth=25
    Color=&FNTimeOfDayColor(
     Night Color=<rgba 0.03 0.0205 0.0014 15>
     Morning Color=<rgba 0.1823 0.0823 0.0202 15>
     Day Color=<rgba 0.1823 0.0823 0.0202 15>
     Evening Color=<rgba 0.1823 0.0823 0.0202 15>
    )
    Lifetime=0.09
    SkeletalMesh=$System.Skeletal Mesh
   ) {module=FN_InitializeAnimTrail002}
  ]
  particleUpdateScript [
   &FN_UpdateAnimTrail(SkeletalMesh=$System.Skeletal Mesh)
   &ScaleColor(
    Scale Alpha=&FloatFromCurve(FloatCurve=DataInterface() {propertyValues {Curve {keys [
         {arriveTangent=-1.8862, interpMode=RCIM_Cubic, leaveTangent=-1.8862, tangentMode=RCTM_User, time=0.2, value=1.25}
         {interpMode=RCIM_Cubic, tangentMode=RCTM_User, time=1, value=0}
        ]}}})
    Scale RGB=&Multiply_Vector(
     A=<1 1 1>
     B=&FN_Select_Vector_from_Time_Of_Day(
      Morning=&MakeVectorFromLinearColorRGB(Color=<rgba 1 1 1 1>)
      Day=&MakeVectorFromLinearColorRGB(Color=<rgba 1 1 1 1>)
      Evening=&MakeVectorFromLinearColorRGB(Color=<rgba 1 1 1 1>)
      Night=&MakeVectorFromLinearColorRGB(Color=<rgba 0.5 0.5 0.5 1>)
     )
    )
    ScaleA=true
    ScaleRGB=true
   ) {enabled=false, module=ScaleColor001}
   &ScaleColor(
    Scale Alpha=&FloatFromCurve(
     FloatCurve=DataInterface() {propertyValues {Curve {keys [
         {interpMode=RCIM_Cubic, tangentMode=RCTM_User, time=0.25, value=1}
         {interpMode=RCIM_Cubic, leaveTangent=-3.732, tangentMode=RCTM_Break, time=0.75, value=1}
         {tangentMode=RCTM_User, time=1, value=0}
        ]}}}
     CurveIndex=&ReturnNormalizedExecIndex()
     Scale Curve=0.33
    )
    ScaleA=true
    ScaleRGB=true
   )
   &ScaleColor(
    Scale Alpha=&FloatFromCurve(
     FloatCurve=DataInterface() {propertyValues {Curve {keys [
         {interpMode=RCIM_Cubic, tangentMode=RCTM_User, time=0.25, value=1}
         {interpMode=RCIM_Cubic, leaveTangent=-0.8441, tangentMode=RCTM_Break, time=0.3, value=1}
         {arriveTangent=-2.0543, tangentMode=RCTM_User, time=1, value=0}
        ]}}}
     CurveIndex=&ReturnNormalizedExecIndex()
     Scale Curve=0.33
    )
    Scale RGB=&Multiply_Vector(
     A=&Multiply_Vector(
      A=<1 1 1>
      B=&FN_Select_Vector_from_Time_Of_Day(
       Morning=&MakeVectorFromLinearColorRGB(Color=<rgba 1 1 1 1>)
       Day=&MakeVectorFromLinearColorRGB(Color=<rgba 1 1 1 1>)
       Evening=&MakeVectorFromLinearColorRGB(Color=<rgba 1 1 1 1>)
       Night=&MakeVectorFromLinearColorRGB(Color=<rgba 0.5 0.5 0.5 1>)
      )
     )
     B=&FN_Select_Vector_from_Time_Of_Day(Morning=<1 1 1>, Day=<1 1 1>, Evening=<0.65 0.65 0.65>, Night=<0.75 0.75 0.75>)
    )
    ScaleA=true
    ScaleRGB=true
   ) {enabled=false, module=ScaleColor002}
   &ScaleRibbonWidth(Ribbon Width Scale=&FloatFromCurve(FloatCurve=DataInterface() {propertyValues {Curve {keys [
         {arriveTangent=5.7119, interpMode=RCIM_Cubic, leaveTangent=5.7119, tangentMode=RCTM_User, time=0, value=0.5}
         {arriveTangentWeight=0.0381, interpMode=RCIM_Cubic, leaveTangentWeight=0.3051, tangentMode=RCTM_User, time=0.1, value=1}
         {arriveTangentWeight=0.3051, interpMode=RCIM_Cubic, leaveTangentWeight=0.0381, tangentMode=RCTM_User, time=0.9, value=1}
         {arriveTangent=-5.7871, interpMode=RCIM_Cubic, leaveTangent=-5.7871, tangentMode=RCTM_User, time=1, value=0.5}
        ]}}}))
  ]
  rendererClasses [&NiagaraRibbonRendererProperties]
  renderers [{rendererClass=&NiagaraRibbonRendererProperties}]
 }
 Outline {
  emitterUpdateScript [
   &EmitterState(Life Cycle Mode=::NewEnumerator1, Loop Behavior=::NewEnumerator1, Loop Duration=1, UseLoopDelay=false)
   &FN_SpawnAnimTrail()
   /SetVariables(Emitter.Random:NiagaraFloat=&RandomRangeFloat(Minimum=0, Maximum=1)) {isSetParameters=true}
  ]
  particleSpawnScript [
   /Game/Effects/Niagara/Modules/AnimTrail/FN_InitializeAnimTrailBoneIndecies_v2/FN_InitializeAnimTrailBoneIndecies(First Bone or Socket Index=5, Second Bone or Socket Index=5)
   &FN_InitializeAnimTrail(AnimTrailWidth=20, Color=<rgba 1 1 1 0.5>, Lifetime=0.11, SkeletalMesh=$System.Skeletal Mesh)
   &DynamicMaterialParameters(
    Index 0 Param 4=$Emitter.Random
    Module.Param0WriteEnabled=false
    Module.Param1WriteEnabled=false
    Module.Param2WriteEnabled=false
    Module.Param3WriteEnabled=true
   ) {module=DynamicMaterialParameters001}
   /SetVariables(Particles.OutlineColor:LinearColor=<rgba 0 0 0 0.9>) {isSetParameters=true}
  ]
  particleUpdateScript [
   &FN_UpdateAnimTrail(SkeletalMesh=$System.Skeletal Mesh) {module=FN_UpdateAnimTrail001}
   &SolveForcesAndVelocity(Clamp Velocity=false, Limit Acceleration=false)
   &ScaleColor(
    Scale Alpha=&FloatFromCurve(FloatCurve=DataInterface() {propertyValues {Curve {keys [
         {arriveTangent=-1.8862, interpMode=RCIM_Cubic, leaveTangent=-1.8862, tangentMode=RCTM_User, time=0.2, value=1.25}
         {interpMode=RCIM_Cubic, tangentMode=RCTM_User, time=1, value=0}
        ]}}})
    Scale RGB=&Multiply_Vector(
     A=<1 1 1>
     B=&FN_Select_Vector_from_Time_Of_Day(
      Morning=&MakeVectorFromLinearColorRGB(Color=<rgba 1 1 1 1>)
      Day=&MakeVectorFromLinearColorRGB(Color=<rgba 1 1 1 1>)
      Evening=&MakeVectorFromLinearColorRGB(Color=<rgba 1 1 1 1>)
      Night=&MakeVectorFromLinearColorRGB(Color=<rgba 0.33 0.33 0.33 1>)
     )
    )
    ScaleA=true
    ScaleRGB=true
   ) {module=ScaleColor001}
   &ScaleRibbonWidth(Ribbon Width Scale=&FloatFromCurve(
     FloatCurve=DataInterface() {propertyValues {Curve {keys [{arriveTangent=-1.7104, interpMode=RCIM_Cubic, leaveTangent=-1.7104, tangentMode=RCTM_User, time=0.4, value=1}, <at 1 0.5 RCIM_Cubic>]}}}
     CurveIndex=$Emitter.NormalizedLoopAge
    )) {enabled=false}
   &ScaleRibbonWidth(Ribbon Width Scale=&FloatFromCurve(
     FloatCurve=DataInterface() {propertyValues {Curve {keys [
         {arriveTangent=6.5238, interpMode=RCIM_Cubic, leaveTangent=6.5238, tangentMode=RCTM_User, time=0, value=0}
         <at 0.1 1>
         {arriveTangent=0.0014, interpMode=RCIM_Cubic, leaveTangent=0.0014, tangentMode=RCTM_User, time=0.4, value=1}
         {arriveTangent=-1.3879, interpMode=RCIM_Cubic, leaveTangent=-1.3879, tangentMode=RCTM_User, time=0.9, value=0.5}
         {arriveTangent=-4.9063, tangentMode=RCTM_User, time=1, value=0}
        ]}}}
     CurveIndex=$Emitter.NormalizedLoopAge
    )) {enabled=false, module=ScaleRibbonWidth001}
   /SetVariables(Particles.OutlineWidth:NiagaraFloat=&Multiply_Float(A=$Particles.RibbonWidth, B=1)) {isSetParameters=true}
  ]
  rendererClasses [&NiagaraRibbonRendererProperties]
  renderers [{rendererClass=&NiagaraRibbonRendererProperties}]
 }
}
summary {
 systemName=NS_Pickaxe_FastCheetah_Trail
 emitters [
  {emitterName=MainTrail, rendererClasses [&NiagaraRibbonRendererProperties]}
  {emitterName=Outline, rendererClasses [&NiagaraRibbonRendererProperties]}
 ]
 userVariables [
  User.PowersActive {type=NiagaraBool}
  User.RibbonIDOffset {type=NiagaraInt32}
 ]
}
